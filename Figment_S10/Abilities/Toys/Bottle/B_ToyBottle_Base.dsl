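; Base Blueprint for the bottle-flip toy, a thrown physics ball that spins in flight, counts flips and judges its landing.
; On near-flat surfaces the owning client compares bottle up with hit normal to pick cap, base or side.
; Upright landings snap, wobble and play victory cues, while side landings fail and spin.
; Server-to-multicast events mirror the result, and a replicated message reports successful flips to the squad feed.
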(event Collision|EventHit
  (params (MyComp :type Primitive%20Component%20Object%20Reference) (Other :type Actor%20Object%20Reference) (OtherComp :type Primitive%20Component%20Object%20Reference) (bSelfMoved :type Boolean) (HitLocation :type Vector :default "0, 0, 0") (HitNormal :type Vector :default "0, 0, 0") (NormalImpulse :type Vector :default "0, 0, 0") (Hit :type Hit%20Result%20Structure%20%28by%20ref%29))
  (|CallParentFunction MyComp Other OtherComp bSelfMoved HitLocation HitNormal NormalImpulse Hit))

(event Visuals|EventInitializeToyPreviewInFrontend
  ;[1 1 1] FrontEnd
  (|CallParentFunction))

(event Custom|CloseGates
  (bind _ball (Variables|Default|GetBall))
  (bind _returnvalue (Transformation|GetUpVector _ball))
  (bind _hitnormal (Variables|Vectors|GetHitNormal))
  (bind _returnvalue_2 (Math|Vector|DotProduct _hitnormal _returnvalue))
  (bind _returnvalue_3 (Math|Trig|Acos(Degrees) _returnvalue_2))
  (bind _capsuccessrange (Variables|Default|GetCapSuccessRange))
  (bind _returnvalue_4 (ScalableFloat|GetValueatLevel _capsuccessrange))
  (bind _returnvalue_5 (+ _returnvalue_4 1.8e+02))
  (bind _returnvalue_6 (* _returnvalue_4 -1.0))
  (bind _returnvalue_7 (+ _returnvalue_6 1.8e+02))
  (bind _returnvalue_8 (Math|Float|InRange(Float) _returnvalue_3 _returnvalue_7 _returnvalue_5))
  (bind _projectilemovement (Variables|Default|GetProjectileMovement))
  ;[1 1 1] How did the bottle land?
  (Utilities|FlowControl|Sequence
    (:then_0
      (Utilities|FlowControl|Gate false
        (:Exit
          ;;[1 1 1] Set Normal Bottle Should Rotate To.
          (Variables|Rotators|SetLandRotation (Math|Rotator|MakeRotator (.roll (Math|Rotator|MakeRotfromZX (Variables|Vectors|GetHitNormal) (Transformation|GetForwardVector (Variables|Default|GetBall)))) (+ (.pitch (Math|Rotator|MakeRotfromZX (Variables|Vectors|GetHitNormal) (Transformation|GetForwardVector (Variables|Default|GetBall)))) 1.8e+02) (.yaw (Transformation|GetWorldRotation (Variables|Default|GetBall)))))
          ;;[1 1 1] OffSet bottle to match collision sphere.
          (Variables|Default|SetOffset "0, 0,-12.000000"))))
    (:then_1
      (Utilities|FlowControl|Gate false
        (:Exit
          (Variables|Rotators|SetLandRotation (Math|Rotator|MakeRotator (.roll (Math|Rotator|MakeRotfromZX (Variables|Vectors|GetHitNormal) (Transformation|GetForwardVector (Variables|Default|GetBall)))) (.pitch (Math|Rotator|MakeRotfromZX (Variables|Vectors|GetHitNormal) (Transformation|GetForwardVector (Variables|Default|GetBall)))) (.yaw (Transformation|GetWorldRotation (Variables|Default|GetBall)))))
          (Variables|Default|SetOffset "0, 0,12.000000"))))
    (:then_2
      (Utilities|FlowControl|Gate false
        (:Exit
          (Variables|Rotators|SetLandRotation (Math|Rotator|MakeRotator (.roll (Transformation|GetWorldRotation (Variables|Default|GetBall))) (+ (.pitch (Math|Rotator|MakeRotfromZX (Variables|Vectors|GetHitNormal) (Transformation|GetForwardVector (Variables|Default|GetBall)))) -9e+01) (.yaw (Transformation|GetWorldRotation (Variables|Default|GetBall)))))
          (CallFunction|CloseGates)
          (Utilities|FlowControl|Sequence
            (:then_0
              (Development|PrintString "BAD" true true "(R=1.000000,G=0.000000,B=0.016751,A=1.000000)" 3.0)
              (Variables|Default|SetSuccess false)
              (CallFunction|ServerFinished (Variables|Default|GetSuccess)))
            (:then_1
              (CallFunction|StopBottleRotationServer)
              (CallFunction|ServerRotatetoNormal)
              (CallFunction|ClientRotatetoNormal)
              ;;[1 1 1] Spin Bottle On Side
              (CallFunction|ServerSpinBottle)))))))
  (CallFunction|CloseGates)
  ;;[1 1 1] Stop and set rotation and offset on Owning Client.
  (Transformation|SetRelativeLocation (Variables|Default|GetBottleParent) (Variables|Default|GetOffset) false false)
  (Variables|/Script/Engine.ProjectileMovementComponent|SetbSimulationEnabled true _projectilemovement)
  (Components|Movement|StopMovementImmediately _projectilemovement)
  (Transformation|SetWorldRotation (Variables|Default|GetBall) (Variables|Rotators|GetLandRotation) false false)
  (Variables|Default|SetActorLocation (Transformation|GetActorLocation))
  ;;[1 1 1] Stop Rotation and Stop Movement Replication
  (CallFunction|ServerOffsetBottle (Variables|Default|GetOffset))
  (CallFunction|StopBottleRotationServer)
  (CallFunction|ServerStopBottleMovement (Transformation|GetActorLocation) (Variables|Rotators|GetLandRotation))
  (Development|PrintString "GOOD" true true "(R=0.000306,G=1.000000,B=0.000000,A=1.000000)" 3.0)
  (Variables|Default|SetSuccess true)
  (Utilities|FlowControl|Sequence
    (:then_1
      (Utilities|FlowControl|Delay (Variables|Default|GetVictoryDelay))
      (Variables|Default|SetDisplayMessage (Utilities|Struct|MakeStructBottleToyMessage (Utilities|Casting|CastToFortPlayerState (Actor|GetOwner)) _returnvalue_8 (Math|Conversions|ToFloat(Integer) (Variables|Default|GetFlipCounter))))
      ;;[1 1 1] Victory Effects
      (CallFunction|VictoryFX _returnvalue_8))
    (:then_2
      (CallFunction|WobbleEventClienttoServer)
      (Utilities|FlowControl|DoOnce false
        (:Completed
          (Variables|Default|SetVector (Transformation|GetWorldLocation (Variables|Default|GetBall)))
          (bind (_direction _newtrack_0) (Timeline|Wobble PlayFromStart)
            (:Update
              (Transformation|SetWorldRotation (Variables|Default|GetBall) (Math|Rotator|MakeRotator (+ (* _newtrack_0 -1.0) (.roll (Variables|Rotators|GetLandRotation))) (+ (.pitch (Variables|Rotators|GetLandRotation)) _newtrack_0) (.yaw (Variables|Rotators|GetLandRotation))) false false)
              (Transformation|SetWorldLocation (Variables|Default|GetBall) (Math|Vector|MakeVector (+ (* _newtrack_0 0.4) (.x (Variables|Default|GetVector))) (+ (.y (Variables|Default|GetVector)) (* _newtrack_0 -0.4)) (.z (Variables|Default|GetVector))) false false))))))))

(event OnProjectileBounce(ProjectileMovement)
  (params (ImpactResult :type Hit%20Result%20Structure%20%28by%20ref%29) (ImpactVelocity :type Vector%20%28by%20ref%29 :default "0, 0, 0"))
  (bind _ball (Variables|Default|GetBall))
  (bind _returnvalue (Transformation|GetUpVector _ball))
  (bind _hitnormal (Variables|Vectors|GetHitNormal))
  (bind _returnvalue_2 (Math|Vector|DotProduct _hitnormal _returnvalue))
  (bind _returnvalue_3 (Math|Trig|Acos(Degrees) _returnvalue_2))
  (bind _capsuccessrange (Variables|Default|GetCapSuccessRange))
  (bind _returnvalue_4 (ScalableFloat|GetValueatLevel _capsuccessrange))
  (bind _returnvalue_5 (+ _returnvalue_4 1.8e+02))
  (bind _returnvalue_6 (* _returnvalue_4 -1.0))
  (bind _returnvalue_7 (+ _returnvalue_6 1.8e+02))
  (bind _returnvalue_8 (Math|Float|InRange(Float) _returnvalue_3 _returnvalue_7 _returnvalue_5))
  (bind _returnvalue_9 (Math|Vector|DotProduct _hitnormal "0, 0,1.000000"))
  (bind _returnvalue_10 (Math|Trig|Acos(Degrees) _returnvalue_9))
  (bind (_bblockinghit _binitialoverlap _time _distance _location _impactpoint _normal _impactnormal) (Collision|BreakHitResult ImpactResult))
  (bind _projectilemovement (Variables|Default|GetProjectileMovement))
  ;[1 1 1] How did the bottle land?
  (Networking|SwitchHasAuthority
    (:Remote
      (Variables|Vectors|SetHitNormal _impactnormal)
      ;;[1 1 1] Check if Valid Surface to land on
      (if (<= _returnvalue_10 12.5)
        (if (Utilities|Equal(Object) (Variables|/Script/Engine.Controller|GetPlayerState (Game|GetPlayerController 0)) (Actor|GetOwner))
          ;;[1 1 1] Check if Landed Upside Down
          (if _returnvalue_8
            (Utilities|FlowControl|DoOnce false
              (:Completed
                (Audio|SpawnSoundAttached "/Figment_S10/Athena/Sounds/Toys/Bottle/Toy_BottleFlip_Land_Top_Cue.Toy_BottleFlip_Land_Top_Cue" (Variables|Transformation|GetRootComponent))
                (Utilities|FlowControl|DoOnce false
                  (:Completed
                    (Utilities|FlowControl|Gate false
                      (:Exit
                        ;;[1 1 1] Set Normal Bottle Should Rotate To.
                        (Variables|Rotators|SetLandRotation (Math|Rotator|MakeRotator (.roll (Math|Rotator|MakeRotfromZX (Variables|Vectors|GetHitNormal) (Transformation|GetForwardVector (Variables|Default|GetBall)))) (+ (.pitch (Math|Rotator|MakeRotfromZX (Variables|Vectors|GetHitNormal) (Transformation|GetForwardVector (Variables|Default|GetBall)))) 1.8e+02) (.yaw (Transformation|GetWorldRotation (Variables|Default|GetBall)))))
                        ;;[1 1 1] OffSet bottle to match collision sphere.
                        (Variables|Default|SetOffset "0, 0,-12.000000")))))))
            (elif (<= _returnvalue_3 (ScalableFloat|GetValueatLevel (Variables|Default|GetSuccessRange)))
              (Utilities|FlowControl|DoOnce false
                (:Completed
                  (Audio|SpawnSoundAttached "/Figment_S10/Athena/Sounds/Toys/Bottle/Toy_BottleFlip_Land_Bottom_Cue.Toy_BottleFlip_Land_Bottom_Cue" (Variables|Transformation|GetRootComponent))
                  (Utilities|FlowControl|DoOnce false
                    (:Completed
                      (Utilities|FlowControl|Gate false
                        (:Exit
                          (Variables|Rotators|SetLandRotation (Math|Rotator|MakeRotator (.roll (Math|Rotator|MakeRotfromZX (Variables|Vectors|GetHitNormal) (Transformation|GetForwardVector (Variables|Default|GetBall)))) (.pitch (Math|Rotator|MakeRotfromZX (Variables|Vectors|GetHitNormal) (Transformation|GetForwardVector (Variables|Default|GetBall)))) (.yaw (Transformation|GetWorldRotation (Variables|Default|GetBall)))))
                          (Variables|Default|SetOffset "0, 0,12.000000")))))))
              (else
                (Utilities|FlowControl|Gate false
                  (:Exit
                    (Variables|Rotators|SetLandRotation (Math|Rotator|MakeRotator (.roll (Transformation|GetWorldRotation (Variables|Default|GetBall))) (+ (.pitch (Math|Rotator|MakeRotfromZX (Variables|Vectors|GetHitNormal) (Transformation|GetForwardVector (Variables|Default|GetBall)))) -9e+01) (.yaw (Transformation|GetWorldRotation (Variables|Default|GetBall)))))
                    (CallFunction|CloseGates)
                    (Utilities|FlowControl|Sequence
                      (:then_0
                        (Development|PrintString "BAD" true true "(R=1.000000,G=0.000000,B=0.016751,A=1.000000)" 3.0)
                        (Variables|Default|SetSuccess false)
                        (CallFunction|ServerFinished (Variables|Default|GetSuccess)))
                      (:then_1
                        (CallFunction|StopBottleRotationServer)
                        (CallFunction|ServerRotatetoNormal)
                        (CallFunction|ClientRotatetoNormal)
                        ;;[1 1 1] Spin Bottle On Side
                        (CallFunction|ServerSpinBottle))))))))
          (CallFunction|CloseGates)
          ;;[1 1 1] Stop and set rotation and offset on Owning Client.
          (Transformation|SetRelativeLocation (Variables|Default|GetBottleParent) (Variables|Default|GetOffset) false false)
          (Variables|/Script/Engine.ProjectileMovementComponent|SetbSimulationEnabled true _projectilemovement)
          (Components|Movement|StopMovementImmediately _projectilemovement)
          (Transformation|SetWorldRotation (Variables|Default|GetBall) (Variables|Rotators|GetLandRotation) false false)
          (Variables|Default|SetActorLocation (Transformation|GetActorLocation))
          ;;[1 1 1] Stop Rotation and Stop Movement Replication
          (CallFunction|ServerOffsetBottle (Variables|Default|GetOffset))
          (CallFunction|StopBottleRotationServer)
          (CallFunction|ServerStopBottleMovement (Transformation|GetActorLocation) (Variables|Rotators|GetLandRotation))
          (Development|PrintString "GOOD" true true "(R=0.000306,G=1.000000,B=0.000000,A=1.000000)" 3.0)
          (Variables|Default|SetSuccess true)
          (Utilities|FlowControl|Sequence
            (:then_1
              (Utilities|FlowControl|Delay (Variables|Default|GetVictoryDelay))
              (Variables|Default|SetDisplayMessage (Utilities|Struct|MakeStructBottleToyMessage (Utilities|Casting|CastToFortPlayerState (Actor|GetOwner)) _returnvalue_8 (Math|Conversions|ToFloat(Integer) (Variables|Default|GetFlipCounter))))
              ;;[1 1 1] Victory Effects
              (CallFunction|VictoryFX _returnvalue_8))
            (:then_2
              (CallFunction|WobbleEventClienttoServer)
              (Utilities|FlowControl|DoOnce false
                (:Completed
                  (Variables|Default|SetVector (Transformation|GetWorldLocation (Variables|Default|GetBall)))
                  (Timeline|Wobble PlayFromStart))))))
        (elif (<= _returnvalue_10 77.5)
          _)))))

(event Custom|WobbleEvent
  ;[1 1 1] How did the bottle land?
  (Utilities|FlowControl|DoOnce false
    (:Completed
      (Variables|Default|SetVector (Transformation|GetWorldLocation (Variables|Default|GetBall)))
      (Timeline|Wobble PlayFromStart))))

(event EventProjectileStop
  (params (ImpactResult :type Hit%20Result%20Structure%20%28by%20ref%29))
  (bind _ball (Variables|Default|GetBall))
  (bind _returnvalue (Transformation|GetUpVector _ball))
  (bind _hitnormal (Variables|Vectors|GetHitNormal))
  (bind _returnvalue_2 (Math|Vector|DotProduct _hitnormal _returnvalue))
  (bind _returnvalue_3 (Math|Trig|Acos(Degrees) _returnvalue_2))
  (bind _capsuccessrange (Variables|Default|GetCapSuccessRange))
  (bind _returnvalue_4 (ScalableFloat|GetValueatLevel _capsuccessrange))
  (bind _returnvalue_5 (+ _returnvalue_4 1.8e+02))
  (bind _returnvalue_6 (* _returnvalue_4 -1.0))
  (bind _returnvalue_7 (+ _returnvalue_6 1.8e+02))
  (bind _returnvalue_8 (Math|Float|InRange(Float) _returnvalue_3 _returnvalue_7 _returnvalue_5))
  (bind _returnvalue_9 (Math|Vector|DotProduct _hitnormal "0, 0,1.000000"))
  (bind _returnvalue_10 (Math|Trig|Acos(Degrees) _returnvalue_9))
  (bind (_bblockinghit _binitialoverlap _time _distance _location _impactpoint _normal _impactnormal) (Collision|BreakHitResult ImpactResult))
  (bind _projectilemovement (Variables|Default|GetProjectileMovement))
  (bind _projectilemovement_2 (Variables|Default|GetProjectileMovement))
  ;[1 1 1] How did the bottle land?
  (Utilities|FlowControl|Sequence
    (:then_0
      (Audio|Components|Audio|FadeOut (Variables|Default|GetInAirAudio) 0.0 0.0 "Linear" :!class /Script/Engine.AudioComponent))
    (:then_1
      (Components|Activation|Activate _projectilemovement false)
      (Components|Movement|SetUpdatedComponent _projectilemovement (Variables|FortPhysicsBall|GetSphereCollision))
      (Variables|/Script/Engine.ProjectileMovementComponent|SetbSimulationEnabled true (Variables|Default|GetProjectileMovement))
      (Collision|SetCollisionResponsetoChannel (Variables|FortPhysicsBall|GetSphereCollision) "ECC_Pawn" "ECR_Ignore")
      (Networking|SwitchHasAuthority
        (:Remote
          (Variables|Vectors|SetHitNormal _impactnormal)
          ;;[1 1 1] Check if Valid Surface to land on
          (if (<= _returnvalue_10 12.5)
            (if (Utilities|Equal(Object) (Variables|/Script/Engine.Controller|GetPlayerState (Game|GetPlayerController 0)) (Actor|GetOwner))
              ;;[1 1 1] Check if Landed Upside Down
              (if _returnvalue_8
                (Utilities|FlowControl|DoOnce false
                  (:Completed
                    (Audio|SpawnSoundAttached "/Figment_S10/Athena/Sounds/Toys/Bottle/Toy_BottleFlip_Land_Top_Cue.Toy_BottleFlip_Land_Top_Cue" (Variables|Transformation|GetRootComponent))
                    (Utilities|FlowControl|DoOnce false
                      (:Completed
                        (Utilities|FlowControl|Gate false
                          (:Exit
                            ;;[1 1 1] Set Normal Bottle Should Rotate To.
                            (Variables|Rotators|SetLandRotation (Math|Rotator|MakeRotator (.roll (Math|Rotator|MakeRotfromZX (Variables|Vectors|GetHitNormal) (Transformation|GetForwardVector (Variables|Default|GetBall)))) (+ (.pitch (Math|Rotator|MakeRotfromZX (Variables|Vectors|GetHitNormal) (Transformation|GetForwardVector (Variables|Default|GetBall)))) 1.8e+02) (.yaw (Transformation|GetWorldRotation (Variables|Default|GetBall)))))
                            ;;[1 1 1] OffSet bottle to match collision sphere.
                            (Variables|Default|SetOffset "0, 0,-12.000000")))))))
                (elif (<= _returnvalue_3 (ScalableFloat|GetValueatLevel (Variables|Default|GetSuccessRange)))
                  (Utilities|FlowControl|DoOnce false
                    (:Completed
                      (Audio|SpawnSoundAttached "/Figment_S10/Athena/Sounds/Toys/Bottle/Toy_BottleFlip_Land_Bottom_Cue.Toy_BottleFlip_Land_Bottom_Cue" (Variables|Transformation|GetRootComponent))
                      (Utilities|FlowControl|DoOnce false
                        (:Completed
                          (Utilities|FlowControl|Gate false
                            (:Exit
                              (Variables|Rotators|SetLandRotation (Math|Rotator|MakeRotator (.roll (Math|Rotator|MakeRotfromZX (Variables|Vectors|GetHitNormal) (Transformation|GetForwardVector (Variables|Default|GetBall)))) (.pitch (Math|Rotator|MakeRotfromZX (Variables|Vectors|GetHitNormal) (Transformation|GetForwardVector (Variables|Default|GetBall)))) (.yaw (Transformation|GetWorldRotation (Variables|Default|GetBall)))))
                              (Variables|Default|SetOffset "0, 0,12.000000")))))))
                  (else
                    (Utilities|FlowControl|Gate false
                      (:Exit
                        (Variables|Rotators|SetLandRotation (Math|Rotator|MakeRotator (.roll (Transformation|GetWorldRotation (Variables|Default|GetBall))) (+ (.pitch (Math|Rotator|MakeRotfromZX (Variables|Vectors|GetHitNormal) (Transformation|GetForwardVector (Variables|Default|GetBall)))) -9e+01) (.yaw (Transformation|GetWorldRotation (Variables|Default|GetBall)))))
                        (CallFunction|CloseGates)
                        (Utilities|FlowControl|Sequence
                          (:then_0
                            (Development|PrintString "BAD" true true "(R=1.000000,G=0.000000,B=0.016751,A=1.000000)" 3.0)
                            (Variables|Default|SetSuccess false)
                            (CallFunction|ServerFinished (Variables|Default|GetSuccess)))
                          (:then_1
                            (CallFunction|StopBottleRotationServer)
                            (CallFunction|ServerRotatetoNormal)
                            (CallFunction|ClientRotatetoNormal)
                            ;;[1 1 1] Spin Bottle On Side
                            (CallFunction|ServerSpinBottle))))))))
              (CallFunction|CloseGates)
              ;;[1 1 1] Stop and set rotation and offset on Owning Client.
              (Transformation|SetRelativeLocation (Variables|Default|GetBottleParent) (Variables|Default|GetOffset) false false)
              (Variables|/Script/Engine.ProjectileMovementComponent|SetbSimulationEnabled true _projectilemovement_2)
              (Components|Movement|StopMovementImmediately _projectilemovement_2)
              (Transformation|SetWorldRotation (Variables|Default|GetBall) (Variables|Rotators|GetLandRotation) false false)
              (Variables|Default|SetActorLocation (Transformation|GetActorLocation))
              ;;[1 1 1] Stop Rotation and Stop Movement Replication
              (CallFunction|ServerOffsetBottle (Variables|Default|GetOffset))
              (CallFunction|StopBottleRotationServer)
              (CallFunction|ServerStopBottleMovement (Transformation|GetActorLocation) (Variables|Rotators|GetLandRotation))
              (Development|PrintString "GOOD" true true "(R=0.000306,G=1.000000,B=0.000000,A=1.000000)" 3.0)
              (Variables|Default|SetSuccess true)
              (Utilities|FlowControl|Sequence
                (:then_1
                  (Utilities|FlowControl|Delay (Variables|Default|GetVictoryDelay))
                  (Variables|Default|SetDisplayMessage (Utilities|Struct|MakeStructBottleToyMessage (Utilities|Casting|CastToFortPlayerState (Actor|GetOwner)) _returnvalue_8 (Math|Conversions|ToFloat(Integer) (Variables|Default|GetFlipCounter))))
                  ;;[1 1 1] Victory Effects
                  (CallFunction|VictoryFX _returnvalue_8))
                (:then_2
                  (CallFunction|WobbleEventClienttoServer)
                  (Utilities|FlowControl|DoOnce false
                    (:Completed
                      (Variables|Default|SetVector (Transformation|GetWorldLocation (Variables|Default|GetBall)))
                      (Timeline|Wobble PlayFromStart))))))
            (elif (<= _returnvalue_10 77.5)
              _)))))))

(event Custom|ServerFinished
  (params (Success :type Boolean))
  ;[1 1 1] How did the bottle land?
  (Default|CallFinished Success self))

(event Custom|WobbleEventClientToServer
  ;[1 1 1] How did the bottle land?
  (CallFunction|WobbleEvent))

(event EventBeginPlay
  (bind _ball (Variables|Default|GetBall))
  ;[1 1 1] Randomized Start Rotation
  (|CallParentFunction)
  (Variables|Default|SetIs%20Throwing true)
  (Networking|SwitchHasAuthority
    (:Authority
      _)
    (:Remote
      _))
  (Utilities|FlowControl|Gate true
    (:Exit
      (Networking|SwitchHasAuthority
        (:Remote
          (Transformation|AddLocalRotation _ball (Math|Rotator|ScaleRotator(Float) (Math|Rotator|MakeRotator (ScalableFloat|GetValueatLevel (Variables|Default|GetSpinRate))) (Utilities|Time|GetWorldDeltaSeconds)) false false)
          (if (Math|Float|NearlyEqual(Float) (.roll (Variables|/Script/Engine.SceneComponent|GetRelativeRotation _ball)) 0.0 (Variables|Default|GetFlipAngleThreshold))
            _)
          (Utilities|FlowControl|DoOnce true
            (:Completed
              (bind _result (Math|Integer|IncrementInt (Variables|Default|GetFlipCounter))
                (:%20%20
                  (Development|PrintString (Utilities|String|ToString(Integer) _result)))))))))))

(event Collision|EventActorBeginOverlap
  (params (OtherActor :type Actor%20Object%20Reference))
  (|CallParentFunction OtherActor)
  (CallFunction|HitPlayer/VehicleBounceSound))

(event EventTick
  (params (DeltaSeconds :type Float%20%28single-precision%29))
  (bind _ball (Variables|Default|GetBall))
  (Variables|Default|SetDeltaTime DeltaSeconds)
  (Utilities|FlowControl|Gate true
    (:Exit
      (Networking|SwitchHasAuthority
        (:Remote
          (Transformation|AddLocalRotation _ball (Math|Rotator|ScaleRotator(Float) (Math|Rotator|MakeRotator (ScalableFloat|GetValueatLevel (Variables|Default|GetSpinRate))) (Utilities|Time|GetWorldDeltaSeconds)) false false)
          (if (Math|Float|NearlyEqual(Float) (.roll (Variables|/Script/Engine.SceneComponent|GetRelativeRotation _ball)) 0.0 (Variables|Default|GetFlipAngleThreshold))
            _)
          (Utilities|FlowControl|DoOnce true
            (:Completed
              (bind _result (Math|Integer|IncrementInt (Variables|Default|GetFlipCounter))
                (:%20%20
                  (Development|PrintString (Utilities|String|ToString(Integer) _result)))))))))))

(event Visuals|EventInitializeToyInstance
  (params (OwningPC :type Fort%20Player%20Controller%20Object%20Reference) (NumTimesSummoned :type Integer))
  (|CallParentFunction OwningPC NumTimesSummoned)
  (Variables|Default|SetHasFinishedSpawning true))

(event Custom|StopBottleRotationServer
  (CallFunction|StopBottleRotationMultiCast))

(event Custom|StopBottleRotationMultiCast
  (bind _ball (Variables|Default|GetBall))
  (Utilities|FlowControl|Gate true
    (:Exit
      (Networking|SwitchHasAuthority
        (:Remote
          (Transformation|AddLocalRotation _ball (Math|Rotator|ScaleRotator(Float) (Math|Rotator|MakeRotator (ScalableFloat|GetValueatLevel (Variables|Default|GetSpinRate))) (Utilities|Time|GetWorldDeltaSeconds)) false false)
          (if (Math|Float|NearlyEqual(Float) (.roll (Variables|/Script/Engine.SceneComponent|GetRelativeRotation _ball)) 0.0 (Variables|Default|GetFlipAngleThreshold))
            _)
          (Utilities|FlowControl|DoOnce true
            (:Completed
              (bind _result (Math|Integer|IncrementInt (Variables|Default|GetFlipCounter))
                (:%20%20
                  (Development|PrintString (Utilities|String|ToString(Integer) _result)))))))))))

(event EventSetBallVelocityAndStartTicking
  (params (NewVelocity :type Vector :default "0, 0, 0"))
  (|CallParentFunction NewVelocity)
  (CallFunction|ClientonThrown))

(event Custom|StopBottleRotation
  (bind _ball (Variables|Default|GetBall))
  (Utilities|FlowControl|Gate true
    (:Exit
      (Networking|SwitchHasAuthority
        (:Remote
          (Transformation|AddLocalRotation _ball (Math|Rotator|ScaleRotator(Float) (Math|Rotator|MakeRotator (ScalableFloat|GetValueatLevel (Variables|Default|GetSpinRate))) (Utilities|Time|GetWorldDeltaSeconds)) false false)
          (if (Math|Float|NearlyEqual(Float) (.roll (Variables|/Script/Engine.SceneComponent|GetRelativeRotation _ball)) 0.0 (Variables|Default|GetFlipAngleThreshold))
            _)
          (Utilities|FlowControl|DoOnce true
            (:Completed
              (bind _result (Math|Integer|IncrementInt (Variables|Default|GetFlipCounter))
                (:%20%20
                  (Development|PrintString (Utilities|String|ToString(Integer) _result)))))))))))

(event Custom|Client%20On%20Thrown
  (Utilities|FlowControl|DoOnce false
    (:Completed
      (Variables|Default|SetIs%20Throwing false)
      (Audio|Components|Audio|FadeIn (Variables|Default|GetInAirAudio) 0.25 1.0 0.0 "Linear" :!class /Script/Engine.AudioComponent))))

(event Custom|ServerRotateToNormal
  ;[1 1 1] LerpToLandingRotation
  (CallFunction|MultiCastRotatetoNormal))

(event Custom|Hit%20Player/Vehicle%20Bounce%20Sound
  ;[1 1 1] Support Collision with Players and Vehicles
  (if (not (Variables|Default|GetIs%20Throwing))
    (Utilities|FlowControl|DoOnce false
      (:Completed
        (Audio|PlaySoundatLocation "/Figment_S10/Athena/Sounds/Toys/Bottle/Toy_BottleFlip_Bounce_Cue.Toy_BottleFlip_Bounce_Cue" (Transformation|GetActorLocation))
        (Utilities|FlowControl|Delay 0.2)))))

(event Custom|ServerSpinBottle
  ;[1 1 1] Spine bottle on side.
  (CallFunction|MultiCastSpinBottle))

(event Custom|ClientRotateToNormal
  ;[1 1 1] LerpToLandingRotation
  (Utilities|FlowControl|Sequence
    (:then_0
      (Variables|Rotators|SetBottleRotation (Transformation|GetWorldRotation (Variables|Default|GetBall))))
    (:then_1
      (bind (_direction _newtrack_0) (Timeline|LerpToRotation PlayFromStart)
        (:Update
          (Transformation|SetWorldRotation (Variables|Default|GetBall) (Math|Rotator|Lerp(Rotator) (Variables|Rotators|GetBottleRotation) (Math|Rotator|CombineRotators (Math|Rotator|MakeRotfromZX (Variables|Vectors|GetHitNormal) (Transformation|GetForwardVector (Variables|Default|GetBall))) "90.000000, 0, 0") _newtrack_0 true) false false))))))

(event Custom|MultiCastSpinBottle
  ;[1 1 1] Spine bottle on side.
  (bind (_direction _newtrack_0) (Timeline|Timeline_0 PlayFromStart)
    (:Update
      (Transformation|AddLocalRotation (Variables|Default|GetBottleParent) (Math|Rotator|ScaleRotator(Float) "360.000000, 0, 0" (* (Utilities|Time|GetWorldDeltaSeconds) _newtrack_0)) false false))))

(event Custom|MultiCastRotateToNormal
  ;[1 1 1] LerpToLandingRotation
  (Utilities|FlowControl|Sequence
    (:then_0
      (Variables|Rotators|SetBottleRotation (Transformation|GetWorldRotation (Variables|Default|GetBall))))
    (:then_1
      (Timeline|LerpToRotation PlayFromStart))))

(event Custom|ServerStopBottleMovement
  (params (OwningClientLocation :type Vector) (Rotation :type Rotator))
  ;[1 1 1] StopBottleMovement
  (CallFunction|StopBottleMovementMultiCast OwningClientLocation Rotation))

(event Custom|ServerOffsetBottle
  (params (NewParam :type Vector))
  ;[1 1 1] Offset Bottle
  (CallFunction|MultiCastOffsetBottle NewParam))

(event Custom|StopBottleMovementMultiCast
  (params (OwningClientLocation :type Vector) (LandRotation :type Rotator))
  (bind _projectilemovement (Variables|Default|GetProjectileMovement))
  ;[1 1 1] StopBottleMovement
  (Variables|/Script/Engine.ProjectileMovementComponent|SetbSimulationEnabled true _projectilemovement)
  (Components|Movement|StopMovementImmediately _projectilemovement)
  (Transformation|SetWorldRotation (Variables|Default|GetBall) LandRotation false false)
  (Variables|Default|SetActorLocation OwningClientLocation))

(event Custom|VictoryFX
  (params (LandedOnCap :type Boolean))
  ;[1 1 1] Victory FX
  (if LandedOnCap
    (CallFunction|PlayCueCenteredonMe (Variables|Default|GetLandedOnCapCue))
    (else
      (CallFunction|PlayCueCenteredonMe (Variables|Default|GetVictoryCue)))))

(event Custom|MultiCastOffsetBottle
  (params (NewParam :type Vector))
  ;[1 1 1] Offset Bottle
  (Transformation|SetRelativeLocation (Variables|Default|GetBottleParent) NewParam false false))

(fn CalcDistanceTraveled
  (params (HitLocation :type Vector))
  (bind _output_get (Variables|Default|SetMetersTraveled (/ (Math|Vector|VectorLengthXY (- HitLocation (Variables|Default|GetInitialSpawnPosition))) 1e+02)))
  (return _output_get))

(fn CreateMessage
  (params (PlayerWhoThrew :type Fort%20Player%20State%20Object%20Reference) (PlayerWhoCaught :type Fort%20Player%20State%20Object%20Reference) (PassDistance :type Float%20%28double-precision%29))
  (if (Variables|/Script/FortniteGame.FortRuntimeOptions|GetbToyMessagingEnabled (Options|GetRuntimeOptions))
    (Variables|Default|SetDisplayMessage)
    (Variables|Default|SetMessageSN (Math|Byte|Byte+Byte (Variables|Default|GetMessageSN) 1))))

(fn OnRep_ActorLocation
  (bind _actorlocation (Variables|Default|GetActorLocation))
  (Transformation|SetActorLocation _actorlocation)
  (Development|PrintString (Utilities|String|ToString(Vector) _actorlocation)))

(fn OnRep_DisplayMessage
  (bind _returnvalue (BlueprintContext|GetFortToyContext))
  (bind (_displaymessage_playerwhothrew_22_ef5d780b4020e911ab429d9ba6b977ba _displaymessage_landedoncap_28_800cc3c44dd84e656fcddc9e415ec97d _displaymessage_passdistance_26_a600d2a048cb01620b914bb4107a6cd5) (Utilities|Struct|BreakStructBottleToyMessage (Variables|GetDisplayMessage_PlayerWhoThrew_22_EF5D780B4020E911AB429D9BA6B977BA)))
  (Networking|SwitchHasAuthority
    (:Remote
      (Utilities|IsValid _displaymessage_playerwhothrew_22_ef5d780b4020e911ab429d9ba6b977ba
        (:Is%20Valid
          (bind _returnvalue_2 (Class|FortToyContext|FormatPlayerNameforDisplay (BlueprintContext|GetFortToyContext) _displaymessage_playerwhothrew_22_ef5d780b4020e911ab429d9ba6b977ba "ToyOwner"))
          (Utilities|IsValid _returnvalue
            (:Is%20Valid
              (Class|FortToyContext|ConditionalPrinttoMessageFeed _returnvalue _displaymessage_playerwhothrew_22_ef5d780b4020e911ab429d9ba6b977ba (Transformation|GetActorLocation) "SquadMembersOnly" (Utilities|Text|FormatText (select _displaymessage_landedoncap_28_800cc3c44dd84e656fcddc9e415ec97d (select (> _displaymessage_passdistance_26_a600d2a048cb01620b914bb4107a6cd5 1.0) (Variables|MessageText|GetLandedOnCapVictoryMessage) (Variables|MessageText|GetLandedOnCapVictoryMessage_OneFlip)) (select (> _displaymessage_passdistance_26_a600d2a048cb01620b914bb4107a6cd5 1.0) (Variables|MessageText|GetVictoryMessage) (Variables|MessageText|GetVictoryMessage_OneFlip))) _returnvalue_2 _displaymessage_passdistance_26_a600d2a048cb01620b914bb4107a6cd5)))))))))

(fn OnRep_IdealLerpEndPos)

(fn OnRep_LandRotation
  (Transformation|SetWorldRotation (Variables|Default|GetBall) (Variables|Rotators|GetLandRotation) false false))

(fn OnRep_Offset)

(fn OnRep_RandomStartRotation
  (Transformation|AddLocalRotation (Variables|Default|GetBall) (Variables|Default|GetRandomStartRotation) false false))

(fn OnRep_Success
  (CallFunction|ServerFinished (Variables|Default|GetSuccess)))

(fn ConstructionScript
  (|CallParentFunction)
  (Networking|SwitchHasAuthority
    (:Remote
      (Variables|/Script/Engine.PrimitiveComponent|SetbGenerateOverlapEvents false (Variables|FortPhysicsBall|GetSphereCollision)))))
